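# Niagara system for the PartyJazzTwinkleToes emote, drawing a flickering dance floor of colored tiles and light beams around the character.
# Tiles and Light_Beams spawn mesh particles on a character-derived, world-space grid, with random pink, blue or purple colors, distance-based scaling and short fades.
# Light_Bottom adds a color-cycling point light near the base.
# User.bIsFrontEndPreview switches the beam emitter's SphereMaskLocation value.
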NiagaraSystem /BRCosmetics/Effects/Fort_Effects/Emotes/PartyJazzTwinkleToes/NS_Emote_PartyJazzTwinkleToes
dependencies {
 usedDataInterfaces [&NiagaraDataInterfaceVectorCurve, &NiagaraDataInterfaceCurve, &NiagaraDataInterfaceColorCurve]
 usedDynamicInputs []
 usedModules [&SystemState]
}
emitters {
 Light_Beams {
  emitterUpdateScript [
   &EmitterState(Inactive Response=::NewEnumerator1, Life Cycle Mode=::NewEnumerator1, Loop Duration=0.5, UseLoopDelay=false)
   /Niagara/Modules/Spawn/Location/SpawnParticlesInGrid(X Count=8, Y Count=8, Z Count=0)
   /SetVariables(Emitter.SphereMaskLocation:NiagaraFloat=&MakeCustomFloatFromBool(Bool=$User.bIsFrontEndPreview, True Float=0, False Float=1)) {isSetParameters=true}
  ]
  particleSpawnScript [
   /Niagara/Modules/Spawn/Initialization/InitializeParticle(
    Color=&MakeCustomLinearColorFromBool(
     Bool=/Niagara/DynamicInputs/Random/RandomBool(Randomness Mode:ENiagaraRandomnessMode=::NewEnumerator0, Random Seed:NiagaraInt32=0, Fixed Random Seed:NiagaraBool=false)
     True Color=&MakeCustomLinearColorFromBool(
      Bool=&RandomBool(Randomness Mode:ENiagaraRandomnessMode=::NewEnumerator0, Random Seed:NiagaraInt32=0, Fixed Random Seed:NiagaraBool=false)
      True Color=<rgba 1 0 0.9464 1>
      False Color=<rgba 0 0.5783 1 1>
     )
     False Color=&MakeCustomLinearColorFromBool(
      Bool=&RandomBool(Randomness Mode:ENiagaraRandomnessMode=::NewEnumerator0, Random Seed:NiagaraInt32=0, Fixed Random Seed:NiagaraBool=false)
      True Color=<rgba 0 0.1751 1 1>
      False Color=<rgba 0.4003 0 1 1>
     )
    )
    Lifetime=&UniformRangedFloat(Minimum=0.5, Maximum=0.5)
    Mass=&UniformRangedFloat(Minimum=0.4, Maximum=2)
    Mesh Scale=<1.6 1.6 1>
    Module.Write Color=true
    Module.Write Lifetime=true
    Module.Write Mass=true
    Module.Write Material Random=true
    Module.Write Position=false
    Module.Write Scale=true
    Module.Write SpriteRotation=true
    Module.Write SpriteSize=false
    Position=&SimulationPosition()
    Sprite Rotation=&NormalizedAngleToDegrees(NormalizedAngle=&UniformRangedFloat(Minimum=0, Maximum=1))
   )
   /Niagara/Modules/Spawn/Location/V2/GridLocation(
    Coordinate Space:ENiagaraCoordinateSpace=::World
    Grid Origin:NiagaraPosition=&self:Grid_Based_Particle_Spawning_FromCharacter(CellSize:NiagaraFloat=18.5, Offset:Vector2f=<0 0>)
    Module.Jitter Inside Grid Cells=false
    Module.Override Random Seed=false
    XYZ Dimensions:Vector3f=<18.5 18.5 0>
   )
  ]
  particleUpdateScript [
   &ParticleState()
   &SolveForcesAndVelocity(Clamp Velocity=false, Limit Acceleration=false)
   &self:DistanceBased_Scaling(Distance=50, Inside Distance=true, ScaleSize:Vector3f=<0.75 0.75 10>) {module=DistanceBased_Scaling}
   &ScaleColor(
    Scale Alpha=&FloatFromCurve(FloatCurve=DataInterface() {propertyValues {Curve {keys [<at 0 0>,<at 0.1 1>,<at 0.9 1>,<at 1 0>]}}})
    Scale RGB=&VectorFromCurve(VectorCurve=DataInterface() {propertyValues {
       XCurve {keys [<at 0 1.5>,<at 0.4635 1.5>,<at 1 1.5>]}
       YCurve {keys [<at 0 1.5>,<at 0.4635 1.5>,<at 1 1.5>]}
       ZCurve {keys [<at 0 1.5>,<at 0.5 0.5>,<at 1 1.5>]}
      }})
    ScaleA=true
    ScaleRGB=true
   )
  ]
  rendererClasses [&NiagaraMeshRendererProperties]
  renderers [{rendererClass=&NiagaraMeshRendererProperties, rendererIndex=0}]
 }
 Light_Bottom {
  emitterUpdateScript [
   &EmitterState(Inactive Response=::NewEnumerator1, Life Cycle Mode=::NewEnumerator1, Loop Duration=2, UseLoopDelay=false)
   &SpawnBurst_Instantaneous(Spawn Count=1, Use Spawn Probability=false)
  ]
  particleSpawnScript [
   &InitializeParticle(
    Color=<rgba 50 50 50 1>
    Lifetime=2
    Mass=&UniformRangedFloat(Minimum=0.4, Maximum=2)
    Mesh Scale=<1.5 1.5 1.5>
    Module.Write Color=true
    Module.Write Lifetime=true
    Module.Write Mass=false
    Module.Write Material Random=true
    Module.Write Position=false
    Module.Write Scale=true
    Module.Write SpriteRotation=true
    Module.Write SpriteSize=false
    Position=&SimulationPosition()
    Sprite Rotation=&NormalizedAngleToDegrees(NormalizedAngle=&UniformRangedFloat(Minimum=0, Maximum=1))
   )
   &SystemLocation(MaskSpawnGroup=false, Offset=<0 0 40>)
   &CylinderLocation(
    Cylinder Height=0
    Cylinder Radius=50
    MaskSpawnGroup=false
    Module.Enable Lathe Profile=false
    Module.Override Randomness=false
    Module.Surface Only=true
    Surface Only Band Thickness=0
    Use Endcaps In Surface Only Mode=false
   )
  ]
  particleUpdateScript [
   &ParticleState()
   &FN_Light_v2(Light Radius=200)
   &Color(
    Color=&ColorFromCurve(ColorCurve=DataInterface() {propertyValues {
       AlphaCurve {keys [<at 0 0 RCIM_Cubic>,<at 0.5025 1>,<at 1 0 RCIM_Cubic>]}
       BlueCurve {keys [<at 0 0 RCIM_Cubic>,<at 0.2534 0.5>,<at 0.4982 0.3343 RCIM_Cubic>,<at 0.8145 0.5>,<at 1 0>]}
       GreenCurve {keys [<at 0 0 RCIM_Cubic>,<at 0.2534 0>,<at 0.4982 0 RCIM_Cubic>,<at 0.8145 0>,<at 1 0>]}
       RedCurve {keys [<at 0 0 RCIM_Cubic>,<at 0.2534 0.0655>,<at 0.4982 0.5 RCIM_Cubic>,<at 0.8145 0.0655>,<at 1 0>]}
      }})
    Scale Color=<5 5 5>
   )
   &SolveForcesAndVelocity(Clamp Velocity=false, Limit Acceleration=false)
   /SetVariables(Particles.Color=&MultiplyLinearColor(Linear Color=$Particles.Color, Scale Factor=<0.3 0.3 0.3 1>)) {isSetParameters=true}
  ]
  rendererClasses [&NiagaraLightRendererProperties]
  renderers [{rendererClass=&NiagaraLightRendererProperties, rendererIndex=0}]
 }
 Tiles {
  emitterUpdateScript [
   &EmitterState(Inactive Response=::NewEnumerator1, Life Cycle Mode=::NewEnumerator1, Loop Duration=0.5, UseLoopDelay=false)
   &SpawnParticlesInGrid(X Count=8, Y Count=8, Z Count=0)
  ]
  particleSpawnScript [
   &InitializeParticle(
    Color=&MakeCustomLinearColorFromBool(
     Bool=&RandomBool(Randomness Mode:ENiagaraRandomnessMode=::NewEnumerator0, Random Seed:NiagaraInt32=0, Fixed Random Seed:NiagaraBool=false)
     True Color=&MakeCustomLinearColorFromBool(
      Bool=&RandomBool(Randomness Mode:ENiagaraRandomnessMode=::NewEnumerator0, Random Seed:NiagaraInt32=0, Fixed Random Seed:NiagaraBool=false)
      True Color=<rgba 1 0 0.9464 1>
      False Color=<rgba 0 0.5783 1 1>
     )
     False Color=&MakeCustomLinearColorFromBool(
      Bool=&RandomBool(Randomness Mode:ENiagaraRandomnessMode=::NewEnumerator0, Random Seed:NiagaraInt32=0, Fixed Random Seed:NiagaraBool=false)
      True Color=<rgba 0 0.1751 1 1>
      False Color=<rgba 0.4003 0 1 1>
     )
    )
    Lifetime=&UniformRangedFloat(Minimum=0.5, Maximum=0.5)
    Mass=&UniformRangedFloat(Minimum=0.4, Maximum=2)
    Mesh Scale=<1.6 1.6 0.1>
    Module.Write Color=true
    Module.Write Lifetime=true
    Module.Write Mass=true
    Module.Write Material Random=true
    Module.Write Position=false
    Module.Write Scale=true
    Module.Write SpriteRotation=true
    Module.Write SpriteSize=false
    Position=&SimulationPosition()
    Sprite Rotation=&NormalizedAngleToDegrees(NormalizedAngle=&UniformRangedFloat(Minimum=0, Maximum=1))
   )
   &GridLocation(
    Coordinate Space:ENiagaraCoordinateSpace=::World
    Grid Origin:NiagaraPosition=&self:Grid_Based_Particle_Spawning_FromCharacter(CellSize:NiagaraFloat=18.5, Offset:Vector2f=<0 0>)
    Module.Jitter Inside Grid Cells=false
    Module.Override Random Seed=false
    XYZ Dimensions:Vector3f=<18.5 18.5 0>
   )
  ]
  particleUpdateScript [
   &ParticleState()
   &SolveForcesAndVelocity(Clamp Velocity=false, Limit Acceleration=false)
   &self:DistanceBased_Scaling(Distance=50, Inside Distance=true, ScaleSize:Vector3f=<0.75 0.75 0.1>) {module=DistanceBased_Scaling}
   &ScaleColor(
    Scale Alpha=&FloatFromCurve(FloatCurve=DataInterface() {propertyValues {Curve {keys [<at 0 0>,<at 0.1 1>,<at 0.9 1>,<at 1 0>]}}})
    Scale RGB=&VectorFromCurve(VectorCurve=DataInterface() {propertyValues {
       XCurve {keys [<at 0 1>,<at 0.4635 1>,<at 1 1>]}
       YCurve {keys [<at 0 1>,<at 0.4635 1>,<at 1 1>]}
       ZCurve {keys [<at 0 1>,<at 0.5 0.5>,<at 1 1>]}
      }})
    ScaleA=true
    ScaleRGB=true
   )
  ]
  rendererClasses [&NiagaraMeshRendererProperties]
  renderers [{rendererClass=&NiagaraMeshRendererProperties, rendererIndex=0}]
 }
}
summary {
 systemName=NS_Emote_PartyJazzTwinkleToes
 emitters [
  {emitterName=Tiles, rendererClasses [&NiagaraMeshRendererProperties]}
  {emitterName=Light_Bottom, rendererClasses [&NiagaraLightRendererProperties]}
  {emitterName=Light_Beams, rendererClasses [&NiagaraMeshRendererProperties]}
 ]
 userVariables [User.bIsFrontEndPreview {type=NiagaraBool}]
}
